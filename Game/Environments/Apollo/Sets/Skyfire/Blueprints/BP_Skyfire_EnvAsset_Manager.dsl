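; This manager Blueprint tells Skyfire environment assets their loot rarity, so loot rooms built from one shared level can be colored differently at runtime.
; It holds a map linking loot rarities to building foundations.
; GetRarityForEnvAsset validates the given actor, loops through that map, and returns the rarity whose foundation contains it.
; When no foundation matches, it falls back to a default enum value.
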(fn GetRarityForEnvAsset
  (params (EnvAsset :type Actor%20Object%20Reference))
  (bind _buildingfoundationtoraritymap (Variables|Default|GetBuildingFoundationToRarityMap))
  ;[0.5312 1 0.8266] BuildingFoundationToRarityMap contains building foundations in and what loot rarity they correspond to. This function checks if the passed in EnvAsset exists in any of those building foundations, and if so returns the rarity that building foundation corresponds to.
  ; This is used to set the skyfire loot rooms' coloring at runtime as each loot room is the same level loaded by multiple building foundations.
  (Utilities|IsValid EnvAsset
    (:Is%20Valid
      (bind _keys (Utilities|Map|Keys _buildingfoundationtoraritymap))
      (for _array_element _keys
        (if (.ReturnValue (Utilities|Map|Find _buildingfoundationtoraritymap _array_element))
          (bind _returnvalue (Foundation|DoesBuildingFoundationContainActor (Utilities|Map|Find _buildingfoundationtoraritymap _array_element) EnvAsset))
          (if _returnvalue
            (return _array_element))))
      (return "NewEnumerator0"))))

(fn ConstructionScript)
