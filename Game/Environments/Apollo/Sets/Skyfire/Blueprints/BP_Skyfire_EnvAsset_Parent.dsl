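; Parent Blueprint for Skyfire environment assets that colors building meshes by loot rarity.
; At BeginPlay it gets rarity from the EnvAsset manager, or uses a default, then applies it; the construction script also applies it, both only where cosmetic events run.
; Each parent material type maps to a custom primitive data index and rarity-based value, with glyphs getting a random value in sixteenth steps.
; A debug event copies its rarity to all instances.
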(event EventBeginPlay
  ;[0.1553 1 0.5413] Event BeginPlay - Check with the Skyfire EnvAsset manager if we should update our rarity and building colors
  (|CallParentFunction)
  (Networking|CanExecuteCosmeticEvents
    (:True
      (bind _returnvalue (Actor|GetActorOfClass "/Game/Environments/Apollo/Sets/Skyfire/Blueprints/BP_Skyfire_EnvAsset_Manager.BP_Skyfire_EnvAsset_Manager_C"))
      (Utilities|IsValid _returnvalue
        (:Is%20Valid
          (bind _rarity (Class|BPSkyfireEnvAssetManager|GetRarityforEnvAsset _returnvalue self))
          (Variables|Default|SetSkyFireRarity _rarity))
        (:Is%20Not%20Valid
          (Variables|Default|SetSkyFireRarity "NewEnumerator0")))
      (CallFunction|SetRarityColorParameters))))

(event Custom|DEBUG_ONLY_SetBuildingRarityColorOnAllBuildingPiecesAndLights
  ;[1 1 1] For debug purposes only in PIE
  (bind _outactors (Actor|GetAllActorsOfClass "/Game/Environments/Apollo/Sets/Skyfire/Blueprints/BP_Skyfire_EnvAsset_Parent.BP_Skyfire_EnvAsset_Parent_C"))
  (for _array_element _outactors
    (Variables|Default|SetSkyFireRarity (Variables|Default|GetSkyFireRarity) _array_element)
    (CallFunction|SetRarityColorParameters _array_element)))

(fn GenerateGlyphRandomValue
  (return (/ (Math|Conversions|ToFloat(Integer) (Math|Float|Round (* (Math|Random|RandomFloat) 16.0))) 16.0)))

(fn GetCPDIndexByParentMaterial
  (params (Skyfire%20Parent%20Material :type B_Enum_Skyfire_LootRoom_ParentMaterials%20Enum :default "NewEnumerator0"))
  (return (Utilities|Select :Index Skyfire%20Parent%20Material :NewEnumerator0 0 :NewEnumerator5 1 :NewEnumerator6 2 :NewEnumerator8 3 :NewEnumerator9 4 :NewEnumerator10 5 :NewEnumerator11 6)))

(fn GetCPDValueByParentMaterial
  (params (MaterialType :type B_Enum_Skyfire_LootRoom_ParentMaterials%20Enum :default "NewEnumerator0") (ContainsMaterialToSet :type Boolean))
  (return (Utilities|Select :Index (select (Variables|GetContainsMaterialToSet) (Variables|Default|GetSkyFireRarity) "NewEnumerator0") :NewEnumerator5 0.0 :NewEnumerator0 (Utilities|Select :Index (Variables|GetMaterialType) :NewEnumerator0 8.0 :NewEnumerator5 46.0 :NewEnumerator6 23.0 :NewEnumerator8 25.0 :NewEnumerator9 36.0 :NewEnumerator10 41.0 :NewEnumerator11 (Variables|Default|GetGlyphRandomValue)) :NewEnumerator1 (Utilities|Select :Index (Variables|GetMaterialType) :NewEnumerator0 22.0 :NewEnumerator5 18.0 :NewEnumerator6 24.0 :NewEnumerator8 26.0 :NewEnumerator9 37.0 :NewEnumerator10 42.0 :NewEnumerator11 (Variables|Default|GetGlyphRandomValue)) :NewEnumerator2 (Utilities|Select :Index (Variables|GetMaterialType) :NewEnumerator0 47.0 :NewEnumerator5 19.0 :NewEnumerator6 27.0 :NewEnumerator8 28.0 :NewEnumerator9 38.0 :NewEnumerator10 43.0 :NewEnumerator11 (Variables|Default|GetGlyphRandomValue)) :NewEnumerator3 (Utilities|Select :Index (Variables|GetMaterialType) :NewEnumerator0 48.0 :NewEnumerator5 2e+01 :NewEnumerator6 29.0 :NewEnumerator8 3e+01 :NewEnumerator9 39.0 :NewEnumerator10 44.0 :NewEnumerator11 (Variables|Default|GetGlyphRandomValue)) :NewEnumerator4 (Utilities|Select :Index (Variables|GetMaterialType) :NewEnumerator0 49.0 :NewEnumerator5 21.0 :NewEnumerator6 31.0 :NewEnumerator8 32.0 :NewEnumerator9 4e+01 :NewEnumerator10 45.0 :NewEnumerator11 (Variables|Default|GetGlyphRandomValue)))))

(fn SetRarityColorParameters
  (if (not (Utilities|Enum|Equal(Enum) (Variables|Default|GetSkyFireRarity) "NewEnumerator5"))
    (for _array_element (Variables|Default|GetMaterialParents)
      (Variables|Default|SetGlyphRandomValue (CallFunction|GenerateGlyphRandomValue))
      (Rendering|Material|SetCustomPrimitiveDataFloat (Variables|Debug|GetStaticMeshComponent) (CallFunction|GetCPDIndexbyParentMaterial _array_element) (CallFunction|GetCPDValuebyParentMaterial _array_element (Utilities|Array|ContainsItem (Variables|Default|GetMaterialParentTypes) _array_element))))))

(fn ConstructionScript
  (Networking|CanExecuteCosmeticEvents
    (:True
      (CallFunction|SetRarityColorParameters))))
